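# Physics asset for the cape and face-accessory part of the male medium CombCraterHat cosmetic, defining secondary motion for hair and hat bones.
# Capsules on head, neck, upper spine, clavicles and upper arms form the body volume.
# Angle-limited chains form back, side and front hair (box-tipped at front), plus three hat pieces based on spine_05.
# Free-rotation constraints cross-link neighboring hair chains and tie back hair to spine_05.
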PhysicsAsset /BRCosmetics/Characters/Player/Male/Medium/Bodies/M_MED_CombCraterHat/Meshes/Parts/M_MED_CombCraterHat_Cape_FaceAcc_Physics
bodies {
 base_hat_bk [{shapeType=Sphere, radius=1, shapeName=base_hat_bk_sphere}]
 base_hat_lb [{shapeType=Sphere, radius=1, shapeName=base_hat_lb_sphere}]
 base_hat_rb [{shapeType=Sphere, radius=1, shapeName=base_hat_rb_sphere}]
 clavicle_l [{shapeType=Capsule, length=10, radius=5, shapeName=clavicle_l_capsule, center=<9.493 -2.0366 4.4155>, rotation=<rot 65.0051 0.5328 0.5944>}]
 clavicle_r [{
   shapeType=Capsule
   length=9.5
   radius=6
   shapeName=clavicle_r_capsule
   center=<-8.8871 0.9351 -3.3092>
   rotation=<rot -75.0043 -179.067 179.0275>
  }]
 dyn_hair_bck_l_01 [{shapeType=Capsule, length=3, radius=2, shapeName=dyn_hair_bck_l_01_capsule, center=<3.5 0 0>, rotation=<rot 90 -179.9999 179.9999>}]
 dyn_hair_bck_l_02 [{shapeType=Capsule, length=4, radius=3, shapeName=dyn_hair_bck_r_03_capsule, center=<4.5 0 0>, rotation=<rot 90 180 -180>}]
 dyn_hair_bck_l_03 [{shapeType=Capsule, length=1.5, radius=3.5, shapeName=dyn_hair_bck_r_03_capsule, center=<3.5 0 0>, rotation=<rot 90 0 0>}]
 dyn_hair_bck_l_04 [{shapeType=Capsule, length=3, radius=4.5, shapeName=dyn_hair_bck_r_03_capsule, center=<2.287 0 0.043>, rotation=<rot 90 180 -90>}]
 dyn_hair_bck_l_05 [{shapeType=Capsule, length=3.5, radius=5, shapeName=dyn_hair_bck_r_03_capsule, center=<2.287 -0.5452 -0.3098>, rotation=<rot 90 180 -90.0002>}]
 dyn_hair_bck_l_base [{shapeType=Capsule, length=0, radius=0.5, shapeName=dyn_hair_bck_l_base_capsule, rotation=<rot 90 0 0>}]
 dyn_hair_bck_r_01 [{shapeType=Capsule, length=3, radius=2, shapeName=dyn_hair_bck_l_01_capsule, center=<3.5 0 0>, rotation=<rot 90 -179.9999 179.9999>}]
 dyn_hair_bck_r_02 [{shapeType=Capsule, length=4, radius=3, shapeName=dyn_hair_bck_r_03_capsule, center=<4.5 0 0>, rotation=<rot 90 180 -180>}]
 dyn_hair_bck_r_03 [{shapeType=Capsule, length=1.5, radius=3.5, shapeName=dyn_hair_bck_r_03_capsule, center=<3.5 0 0>, rotation=<rot 90 0 0>}]
 dyn_hair_bck_r_04 [{shapeType=Capsule, length=4, radius=4, shapeName=dyn_hair_bck_r_03_capsule, center=<2.287 0 0>, rotation=<rot 90 1.2255 91.2255>}]
 dyn_hair_bck_r_05 [{shapeType=Capsule, length=3, radius=4.5, shapeName=dyn_hair_bck_r_03_capsule, center=<2.287 0 0>, rotation=<rot 90 1.2255 -88.7745>}]
 dyn_hair_bck_r_base [{shapeType=Capsule, length=0, radius=0.5, shapeName=dyn_hair_bck_r_base_capsule, rotation=<rot 90 0 0>}]
 dyn_hair_fr_l_01 [{
   shapeType=Capsule
   length=6.3555
   radius=2.8285
   shapeName=dyn_hair_fr_l_01_capsule
   center=<6.5943 -0.5387 -1.6145>
   rotation=<rot 76.5935 -157.6571 31.2486>
  }]
 dyn_hair_fr_l_02 [{
   shapeType=Capsule
   length=7.5
   radius=3
   shapeName=dyn_hair_fr_r_02_capsule
   center=<3.7209 0.7362 0.9977>
   rotation=<rot 75.7997 -86.3939 -147.0384>
  }]
 dyn_hair_fr_l_03 [{
   shapeType=Box
   extentX=9.5
   extentY=11.5
   extentZ=3.5
   shapeName=dyn_hair_fr_r_03_box
   center=<3.5473 -0.396 0.2>
   rotation=<rot -5.2135 0.0384 -0.2369>
  }]
 dyn_hair_fr_r_01 [{
   shapeType=Capsule
   length=7
   radius=2.5
   shapeName=dyn_hair_fr_l_01_capsule
   center=<9.7112 2.6423 -0.9086>
   rotation=<rot 70.3227 163.1425 -25.0068>
  }]
 dyn_hair_fr_r_02 [{
   shapeType=Capsule
   length=8
   radius=2
   shapeName=dyn_hair_fr_r_02_capsule
   center=<4.8533 -0.2294 0.7611>
   rotation=<rot 83.2965 138.7106 18.297>
  }]
 dyn_hair_fr_r_03 [{
   shapeType=Box
   extentX=9.5
   extentY=11.5
   extentZ=3.5
   shapeName=dyn_hair_fr_r_03_box
   center=<3.332 0.2403 -0.2319>
   rotation=<rot -14.9987 0 20>
  }]
 dyn_hair_side_l_01 [{shapeType=Capsule, length=3.5, radius=2, shapeName=dyn_hair_side_l_01_capsule, center=<3.5 0 0>, rotation=<rot 90 0 0>}]
 dyn_hair_side_l_02 [{shapeType=Capsule, length=4, radius=3, shapeName=dyn_hair_side_l_02_capsule, center=<4.5 0 0>, rotation=<rot 90 0 0>}]
 dyn_hair_side_l_03 [{shapeType=Capsule, length=1.5, radius=3.5, shapeName=dyn_hair_side_l_03_capsule, center=<3.5 0 0>, rotation=<rot 90 0 0>}]
 dyn_hair_side_l_base [{shapeType=Capsule, length=0, radius=0.5, shapeName=dyn_hair_side_l_base_capsule, rotation=<rot 90 0 0>}]
 dyn_hair_side_r_01 [{shapeType=Capsule, length=3.5, radius=2, shapeName=dyn_hair_side_r_01_capsule, center=<3.5 0 0>, rotation=<rot 90 0 0>}]
 dyn_hair_side_r_02 [{
   shapeType=Capsule
   length=4
   radius=3
   shapeName=dyn_hair_side_r_02_capsule
   center=<4.5 0.5076 0.6692>
   rotation=<rot 51.245 89.9999 89.9999>
  }]
 dyn_hair_side_r_03 [{shapeType=Capsule, length=1.5, radius=3.5, shapeName=dyn_hair_side_r_03_capsule, center=<3.5 0 0.4948>, rotation=<rot 64.5174 0 0>}]
 dyn_hair_side_r_base [{shapeType=Capsule, length=0, radius=0.5, shapeName=dyn_hair_side_r_base_capsule, rotation=<rot 90 0 0>}]
 dyn_hat_bk [{shapeType=Capsule, length=14.5, radius=3.5, shapeName=dyn_hat_bk_capsule, center=<5.8136 0.0595 -2>, rotation=<rot 0 0 90>}]
 dyn_hat_lb [{shapeType=Capsule, length=14.5, radius=3.5, shapeName=dyn_hat_lb_capsule, center=<5.8136 0.1844 -2>, rotation=<rot 0 0 81.0471>}]
 dyn_hat_rb [{shapeType=Capsule, length=15, radius=3, shapeName=dyn_hat_rb_capsule, center=<5.8136 0 -2>, rotation=<rot 0 0 93.6239>}]
 hair_fr_l_base [{shapeType=Capsule, length=0.505, radius=0.505, shapeName=hair_fr_l_base_capsule, rotation=<rot 0 -90 0>}]
 hair_fr_r_base [{shapeType=Capsule, length=0.505, radius=0.505, shapeName=hair_fr_r_base_capsule, rotation=<rot 0 -90 0>}]
 head [{shapeType=Capsule, length=7, radius=8, shapeName=head_capsule, center=<2.827 3.6062 0>, rotation=<rot 0 -78.8774 -90>}]
 neck_01 [{shapeType=Capsule, length=0.5, radius=6, shapeName=neck_01_capsule, rotation=<rot 0 -0.0003 0>}]
 neck_02 [{shapeType=Capsule, length=2.5, radius=5.5, shapeName=neck_02_capsule, rotation=<rot 90 164.8051 164.8053>}]
 spine_04 [{shapeType=Capsule, length=17.005, radius=12.5, shapeName=spine_04_capsule, center=<3.6309 0.2268 0>, rotation=<rot 0 -90 0>}]
 spine_05 [{shapeType=Capsule, length=19.5, radius=9, shapeName=spine_05_capsule, center=<2.3988 -0.8368 0>, rotation=<rot 0 -0.0002 0>}]
 upperarm_l [{shapeType=Capsule, length=23.5, radius=5.5, shapeName=upperarm_l_capsule, center=<8.3369 0 0>, rotation=<rot 90 0 0>}]
 upperarm_r [{shapeType=Capsule, length=25, radius=7, shapeName=upperarm_r_capsule, center=<-13 0 0>, rotation=<rot -90 0 0>}]
}
constraints [
 {bone1Name=dyn_hair_bck_l_base, bone2Name=head, swing1LimitDegrees=10, swing2LimitDegrees=45, twistLimitDegrees=2}
 {bone1Name=dyn_hair_bck_l_01, bone2Name=dyn_hair_bck_l_base, swing1LimitDegrees=10, swing2LimitDegrees=45, twistLimitDegrees=2}
 {bone1Name=dyn_hair_bck_l_02, bone2Name=dyn_hair_bck_l_01, swing1LimitDegrees=10, swing2LimitDegrees=40, twistLimitDegrees=2}
 {bone1Name=dyn_hair_bck_l_03, bone2Name=dyn_hair_bck_l_02, swing1LimitDegrees=10, swing2LimitDegrees=40, twistLimitDegrees=3}
 {bone1Name=dyn_hair_bck_l_04, bone2Name=dyn_hair_bck_l_03, swing1LimitDegrees=10, swing2LimitDegrees=40, twistLimitDegrees=2}
 {bone1Name=dyn_hair_bck_l_05, bone2Name=dyn_hair_bck_l_04, swing1LimitDegrees=10, swing2LimitDegrees=45, twistLimitDegrees=20}
 {bone1Name=dyn_hair_bck_r_base, bone2Name=head, swing1LimitDegrees=10, swing2LimitDegrees=45, twistLimitDegrees=2}
 {bone1Name=dyn_hair_bck_r_01, bone2Name=dyn_hair_bck_r_base, swing1LimitDegrees=10, swing2LimitDegrees=45, twistLimitDegrees=2}
 {bone1Name=dyn_hair_bck_r_02, bone2Name=dyn_hair_bck_r_01, swing1LimitDegrees=10, swing2LimitDegrees=40, twistLimitDegrees=2}
 {bone1Name=dyn_hair_bck_r_03, bone2Name=dyn_hair_bck_r_02, swing1LimitDegrees=10, swing2LimitDegrees=40, twistLimitDegrees=2}
 {bone1Name=dyn_hair_bck_r_04, bone2Name=dyn_hair_bck_r_03, swing1LimitDegrees=10, swing2LimitDegrees=40, twistLimitDegrees=2}
 {bone1Name=dyn_hair_bck_r_05, bone2Name=dyn_hair_bck_r_04, swing1LimitDegrees=10, swing2LimitDegrees=45, twistLimitDegrees=2}
 {bone1Name=dyn_hair_side_l_base, bone2Name=head, swing1LimitDegrees=10, swing2LimitDegrees=45, twistLimitDegrees=45}
 {bone1Name=dyn_hair_side_l_01, bone2Name=dyn_hair_side_l_base, swing1LimitDegrees=20, swing2LimitDegrees=20, twistLimitDegrees=2}
 {bone1Name=dyn_hair_side_l_02, bone2Name=dyn_hair_side_l_01, swing1LimitDegrees=10, swing2LimitDegrees=10, twistLimitDegrees=2}
 {bone1Name=dyn_hair_side_l_03, bone2Name=dyn_hair_side_l_02, swing1LimitDegrees=10, swing2LimitDegrees=10, twistLimitDegrees=2}
 {bone1Name=dyn_hair_side_r_base, bone2Name=head, swing1LimitDegrees=10, swing2LimitDegrees=45, twistLimitDegrees=45}
 {bone1Name=dyn_hair_side_r_01, bone2Name=dyn_hair_side_r_base, swing1LimitDegrees=20, swing2LimitDegrees=40, twistLimitDegrees=2}
 {bone1Name=dyn_hair_side_r_02, bone2Name=dyn_hair_side_r_01, swing1LimitDegrees=15, swing2LimitDegrees=30, twistLimitDegrees=2}
 {bone1Name=dyn_hair_side_r_03, bone2Name=dyn_hair_side_r_02, swing1LimitDegrees=30, swing2LimitDegrees=30, twistLimitDegrees=2}
 {
  bone1Name=dyn_hair_bck_l_04
  bone2Name=dyn_hair_bck_r_04
  swing1LimitDegrees=45
  swing1Motion=Free
  swing2LimitDegrees=45
  swing2Motion=Free
  twistLimitDegrees=45
  twistMotion=Free
 }
 {
  bone1Name=dyn_hair_side_r_03
  bone2Name=dyn_hair_bck_r_03
  swing1LimitDegrees=45
  swing1Motion=Free
  swing2LimitDegrees=45
  swing2Motion=Free
  twistLimitDegrees=45
  twistMotion=Free
 }
 {
  bone1Name=dyn_hair_side_l_03
  bone2Name=dyn_hair_bck_l_03
  swing1LimitDegrees=45
  swing1Motion=Free
  swing2LimitDegrees=45
  swing2Motion=Free
  twistLimitDegrees=45
  twistMotion=Free
 }
 {
  bone1Name=dyn_hair_side_l_02
  bone2Name=dyn_hair_bck_l_02
  swing1LimitDegrees=45
  swing1Motion=Free
  swing2LimitDegrees=45
  swing2Motion=Free
  twistLimitDegrees=45
  twistMotion=Free
 }
 {
  bone1Name=dyn_hair_bck_l_05
  bone2Name=dyn_hair_bck_r_05
  swing1LimitDegrees=45
  swing1Motion=Free
  swing2LimitDegrees=45
  swing2Motion=Free
  twistLimitDegrees=45
  twistMotion=Free
 }
 {
  bone1Name=dyn_hair_bck_l_04
  bone2Name=spine_05
  swing1LimitDegrees=45
  swing1Motion=Free
  swing2LimitDegrees=45
  swing2Motion=Free
  twistLimitDegrees=45
  twistMotion=Free
 }
 {
  bone1Name=dyn_hair_bck_r_04
  bone2Name=spine_05
  swing1LimitDegrees=45
  swing1Motion=Free
  swing2LimitDegrees=45
  swing2Motion=Free
  twistLimitDegrees=45
  twistMotion=Free
 }
 {
  bone1Name=hair_fr_r_base
  bone2Name=neck_01
  swing1LimitDegrees=45
  swing1Motion=Free
  swing2LimitDegrees=45
  swing2Motion=Free
  twistLimitDegrees=45
  twistMotion=Free
 }
 {bone1Name=dyn_hair_fr_l_01, bone2Name=hair_fr_l_base, swing1LimitDegrees=25, swing2LimitDegrees=40, twistLimitDegrees=5}
 {
  bone1Name=hair_fr_l_base
  bone2Name=head
  swing1LimitDegrees=45
  swing1Motion=Free
  swing2LimitDegrees=45
  swing2Motion=Free
  twistLimitDegrees=45
  twistMotion=Free
 }
 {bone1Name=dyn_hair_fr_l_02, bone2Name=dyn_hair_fr_l_01, swing1LimitDegrees=20, swing2LimitDegrees=40, twistLimitDegrees=45}
 {bone1Name=dyn_hair_fr_l_03, bone2Name=dyn_hair_fr_l_02, swing1LimitDegrees=20, swing2LimitDegrees=45, twistLimitDegrees=30}
 {bone1Name=dyn_hair_fr_r_01, bone2Name=hair_fr_r_base, swing1LimitDegrees=20, swing2LimitDegrees=40, twistLimitDegrees=10}
 {bone1Name=dyn_hair_fr_r_02, bone2Name=dyn_hair_fr_r_01, swing1LimitDegrees=10, swing2LimitDegrees=40, twistLimitDegrees=45}
 {bone1Name=dyn_hair_fr_r_03, bone2Name=dyn_hair_fr_r_02, swing1LimitDegrees=40, swing2LimitDegrees=40, twistLimitDegrees=30}
 {bone1Name=dyn_hat_bk, bone2Name=base_hat_bk, swing1LimitDegrees=1, swing2LimitDegrees=30, twistLimitDegrees=10}
 {
  bone1Name=base_hat_bk
  bone2Name=spine_05
  swing1LimitDegrees=45
  swing1Motion=Free
  swing2LimitDegrees=45
  swing2Motion=Free
  twistLimitDegrees=45
  twistMotion=Free
 }
 {bone1Name=dyn_hat_lb, bone2Name=base_hat_lb, swing1LimitDegrees=1, swing2LimitDegrees=20, twistLimitDegrees=1}
 {
  bone1Name=base_hat_lb
  bone2Name=spine_05
  swing1LimitDegrees=45
  swing1Motion=Free
  swing2LimitDegrees=45
  swing2Motion=Free
  twistLimitDegrees=45
  twistMotion=Free
 }
 {bone1Name=dyn_hat_rb, bone2Name=base_hat_rb, swing1LimitDegrees=1, swing2LimitDegrees=30, twistLimitDegrees=1}
 {
  bone1Name=base_hat_rb
  bone2Name=spine_05
  swing1LimitDegrees=45
  swing1Motion=Free
  swing2LimitDegrees=45
  swing2Motion=Free
  twistLimitDegrees=45
  twistMotion=Free
 }
 {
  bone1Name=dyn_hair_bck_l_02
  bone2Name=dyn_hair_bck_r_02
  swing1LimitDegrees=45
  swing1Motion=Free
  swing2LimitDegrees=45
  swing2Motion=Free
  twistLimitDegrees=45
  twistMotion=Free
 }
]
