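# Physics asset for the hat and face-accessory mesh of the male medium CombCraterHat player cosmetic, covering dynamic hat and hair bones.
# Three capsule hat bodies (back, left, right) are constrained with narrow angular limits to small base spheres, which link to spine_05 with free rotation.
# Hair capsules also link to spine_05 with free rotation, except the second right-side segment, which chains to the first.
# Spine and unconstrained clavicle capsules provide upper-torso shapes.
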PhysicsAsset /BRCosmetics/Characters/Player/Male/Medium/Bodies/M_MED_CombCraterHat/Meshes/Parts/M_MED_CombCraterHat_Hat_FaceAcc_Physics
bodies {
 base_hat_bk [{shapeType=Sphere, radius=1, shapeName=base_hat_bk_sphere}]
 base_hat_lb [{shapeType=Sphere, radius=1, shapeName=base_hat_lb_sphere}]
 base_hat_rb [{shapeType=Sphere, radius=1, shapeName=base_hat_rb_sphere}]
 clavicle_l [{shapeType=Capsule, length=20.5, radius=10.005, shapeName=clavicle_r_capsule, center=<8.641 0 0.2213>, rotation=<rot -0.0001 -90 88.5329>}]
 clavicle_r [{shapeType=Capsule, length=20.5, radius=10.005, shapeName=clavicle_r_capsule, center=<-8.9463 0 -0.2291>, rotation=<rot -0.0001 -90 88.5329>}]
 dyn_hair_bck_l_01 [{shapeType=Capsule, length=8.5, radius=3.005, shapeName=dyn_hair_bck_l_01_capsule, center=<11.7254 0 -3.4373>, rotation=<rot 0 -90 68.5015>}]
 dyn_hair_bck_r_01 [{shapeType=Capsule, length=8.5, radius=3.005, shapeName=dyn_hair_bck_r_01_capsule, center=<11.7484 0 -3.4957>, rotation=<rot 0 -90 68.5015>}]
 dyn_hair_side_l_01 [{
   shapeType=Capsule
   length=6.505
   radius=3.005
   shapeName=dyn_hair_side_r_01_capsule
   center=<10.6165 1.8919 -0.7309>
   rotation=<rot 0.9196 -108.3776 -87.2343>
  }]
 dyn_hair_side_r_01 [{
   shapeType=Capsule
   length=7.005
   radius=2.505
   shapeName=dyn_hair_side_r_01_capsule
   center=<10.6833 0 -4.1508>
   rotation=<rot 0 -90 -87.0855>
  }]
 dyn_hair_side_r_02 [{
   shapeType=Capsule
   length=5.3843
   radius=3.2668
   shapeName=dyn_hair_side_r_02_capsule
   center=<6.3748 -0.3108 -0.9765>
   rotation=<rot 0 95.5959 -68.6436>
  }]
 dyn_hat_bk [{shapeType=Capsule, length=14.5, radius=3.5, shapeName=dyn_hat_bk_capsule, center=<5.8136 0.0595 -1.794>, rotation=<rot 0 0 90>}]
 dyn_hat_lb [{shapeType=Capsule, length=14.5, radius=3.5, shapeName=dyn_hat_lb_capsule, center=<5.8136 0.1844 -0.7924>, rotation=<rot 0 0 81.0471>}]
 dyn_hat_rb [{shapeType=Capsule, length=15, radius=3, shapeName=dyn_hat_rb_capsule, center=<5.8136 0.0399 -0.3098>, rotation=<rot 0 0 93.6239>}]
 spine_05 [{shapeType=Capsule, length=3, radius=20.005, shapeName=spine_05_capsule, rotation=<rot 0 -90 0>}]
}
constraints [
 {bone1Name=dyn_hat_bk, bone2Name=base_hat_bk, swing1LimitDegrees=1, swing2LimitDegrees=30, twistLimitDegrees=10}
 {bone1Name=dyn_hat_lb, bone2Name=base_hat_lb, swing1LimitDegrees=1, swing2LimitDegrees=20, twistLimitDegrees=1}
 {bone1Name=dyn_hat_rb, bone2Name=base_hat_rb, swing1LimitDegrees=1, swing2LimitDegrees=30, twistLimitDegrees=1}
 {
  bone1Name=base_hat_rb
  bone2Name=spine_05
  swing1LimitDegrees=45
  swing1Motion=Free
  swing2LimitDegrees=45
  swing2Motion=Free
  twistLimitDegrees=45
  twistMotion=Free
 }
 {
  bone1Name=base_hat_lb
  bone2Name=spine_05
  swing1LimitDegrees=45
  swing1Motion=Free
  swing2LimitDegrees=45
  swing2Motion=Free
  twistLimitDegrees=45
  twistMotion=Free
 }
 {
  bone1Name=base_hat_bk
  bone2Name=spine_05
  swing1LimitDegrees=45
  swing1Motion=Free
  swing2LimitDegrees=45
  swing2Motion=Free
  twistLimitDegrees=45
  twistMotion=Free
 }
 {
  bone1Name=dyn_hair_side_r_01
  bone2Name=spine_05
  swing1LimitDegrees=45
  swing1Motion=Free
  swing2LimitDegrees=45
  swing2Motion=Free
  twistLimitDegrees=45
  twistMotion=Free
 }
 {
  bone1Name=dyn_hair_bck_r_01
  bone2Name=spine_05
  swing1LimitDegrees=45
  swing1Motion=Free
  swing2LimitDegrees=45
  swing2Motion=Free
  twistLimitDegrees=45
  twistMotion=Free
 }
 {
  bone1Name=dyn_hair_bck_l_01
  bone2Name=spine_05
  swing1LimitDegrees=45
  swing1Motion=Free
  swing2LimitDegrees=45
  swing2Motion=Free
  twistLimitDegrees=45
  twistMotion=Free
 }
 {
  bone1Name=dyn_hair_side_l_01
  bone2Name=spine_05
  swing1LimitDegrees=45
  swing1Motion=Free
  swing2LimitDegrees=45
  swing2Motion=Free
  twistLimitDegrees=45
  twistMotion=Free
 }
 {
  bone1Name=dyn_hair_side_r_02
  bone2Name=dyn_hair_side_r_01
  swing1LimitDegrees=45
  swing1Motion=Free
  swing2LimitDegrees=45
  swing2Motion=Free
  twistLimitDegrees=45
  twistMotion=Free
 }
]
